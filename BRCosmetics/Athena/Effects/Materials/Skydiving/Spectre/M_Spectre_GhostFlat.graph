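# Translucent, unlit, two-sided material for the Spectre skydiving effect, drawing flat ghostly faces on particle sprites and beam trails.
# A tintable Color parameter runs through an LDR-to-HDR intensity boost and emissive adjustments for the glow.
# A separate mask multiplies particle alpha, the green channel of a tiled, offset spooky-faces texture, and a soft radial SphereMask that fades the sprite edges.
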Material /BRCosmetics/Athena/Effects/Materials/Skydiving/Spectre/M_Spectre_GhostFlat
graph {
 properties {
  bUsedWithBeamTrails=true
  bUsedWithNiagaraSprites=true
  bUsedWithParticleSprites=true
  blendMode=BLEND_Translucent
  pixelDepthOffsetMode=PDOM_Legacy
  shadingModel=MSM_Unlit
  twoSided=true
  usedShadingModels=Unlit
 }
 nodes [
  n1 = VectorParameter() {parameterName=Color, defaultValue=<rgba 0.1 1 0.6643 1>}
  MaterialFunctionCall(Emissive=MaterialFunctionCall(LDR Input=Clamp(None=@n1.RGB), HDR Tint=@n1.RGB, HDR Intensity Multiplier=ScalarParameter() {defaultValue=1.25, parameterName=HDRIntensity}) {materialFunction=&DeriveHDRfromLDR}.RemappedRange) {materialFunction=&MF_FX_EmissiveAdjustments}
  n2 = ScalarParameter() {defaultValue=1, parameterName=UTile}
  n3 = ScalarParameter() {defaultValue=1, parameterName=VTile}
  n4 = Add(
   A=Multiply(A=TextureCoordinate(), B=AppendVector(A=@n2, B=@n3))
   B=AppendVector(A=ScalarParameter() {defaultValue=0.2, parameterName=UOffset}, B=ScalarParameter() {defaultValue=0.15, parameterName=VOffset})
  )
  Multiply(
   A=Multiply(A=ParticleColor().A, B=TextureSample(UVs=@n4) {texture=/Game/Effects/Fort_Effects/Textures/Tile/T_SpookyFaces_Chroma_Tile}.G)
   B=SphereMask(A=TextureCoordinate(), B=Constant() {r=0.5}, Radius=ScalarParameter() {defaultValue=0.4, parameterName=Radius}) {hardnessPercent=0}
  )
 ]
}
