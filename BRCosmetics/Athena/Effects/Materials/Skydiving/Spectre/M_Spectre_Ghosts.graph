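# Additive unlit mesh-particle material, named for the Spectre skydiving effect, that renders ghostly scrolling spooky faces.
# A panned face texture blends a dark color with a time-of-day tint, then passes through HDR and FX emissive adjustments.
# Sine-shaped UV edge masks, particle alpha and near-camera fade shape visibility.
# Vertex offsets combine UV-end flares along normals with scrolling world-space noise, reduced near the object center and driven by dynamic parameters.
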Material /BRCosmetics/Athena/Effects/Materials/Skydiving/Spectre/M_Spectre_Ghosts
graph {
 properties {
  bUsedWithMeshParticles=true
  bUsedWithNiagaraMeshParticles=true
  blendMode=BLEND_Additive
  pixelDepthOffsetMode=PDOM_Legacy
  shadingModel=MSM_Unlit
  usedShadingModels=Unlit
 }
 nodes [
  n1 = VertexNormalWS()
  n2 = ComponentMask(None=TextureCoordinate()) {g=true}
  n3 = Multiply(
   A=Power(Base=OneMinus(None=@n2), Exp=ScalarParameter() {defaultValue=15, parameterName=WPO_ContrastEdge})
   B=ScalarParameter() {defaultValue=300, parameterName=WPOAmount}
  )
  n4 = ComponentMask(None=TextureCoordinate()) {g=true}
  n5 = ScalarParameter() {defaultValue=7, parameterName=WPO_ContrastEdgeBottom}
  n6 = Multiply(A=VertexNormalWS(), B=Multiply(A=Power(Base=@n4, Exp=@n5), B=ScalarParameter() {defaultValue=-500, parameterName=WPOAmountTop}))
  n7 = Constant3Vector() {constant=<rgba 1 1 0 0>}
  n8 = Time()
  n9 = DynamicParameter() {defaultValue=<rgba 1 1 0 420>}
  n10 = ScalarParameter() {parameterName=Speed}
  n11 = Subtract(A=Add(A=Multiply(A=@n8, B=@n9.Param3), B=WorldPosition().XYZ), B=Multiply(A=ObjectPositionWS(), B=Add(A=@n10, B=@n9.Param1)) {constB=0.95})
  n12 = ScalarParameter() {defaultValue=1100, parameterName=Tilling}
  n13 = TextureSample(UVs=Divide(A=ComponentMask(None=@n11) {b=true}, B=Add(A=@n12, B=@n9.Param2))) {texture=&T_NoisePack_01}
  n14 = DynamicParameter() {defaultValue=<rgba 1 1 0 420>}
  n15 = SphereMask(A=WorldPosition().XYZ, B=ObjectPositionWS(), Radius=ScalarParameter() {defaultValue=2000}, Hardness=Constant())
  Add(
   A=Add(A=Multiply(A=@n1, B=@n3), B=Multiply(A=@n6, B=@n7))
   B=Multiply(A=Multiply(A=@n13.RGB, B=@n14.Param4) {constB=1500}, B=OneMinus(None=@n15))
  )
  n16 = VectorParameter() {parameterName=Color, defaultValue=<rgba 0.2 2 1.3286 1>}
  n17 = TextureCoordinate()
  n18 = AppendVector(A=ScalarParameter() {defaultValue=1, parameterName=UTile}, B=ScalarParameter() {defaultValue=7, parameterName=VTile})
  n19 = Time()
  n20 = ScalarParameter() {defaultValue=3.5, parameterName=Panner Speed}
  n21 = TextureSample(UVs=Panner(Coordinate=Multiply(A=@n17, B=@n18), Time=Multiply(A=@n19, B=@n20)) {speedY=1}) {texture=/Game/Effects/Fort_Effects/Textures/Tile/T_SpookyFaces_Chroma_Tile}
  n22 = LinearInterpolate(
   A=VectorParameter() {parameterName=DarkColor, defaultValue=<rgba 0.125 0.3477 0.6 0>}.RGB
   B=MaterialFunctionCall(
    Color_Morning=@n16.RGB
    Color_Day=VectorParameter() {parameterName=Color_Day, defaultValue=<rgba 0.234 1 0.7143 1>}.RGB
    Color_Evening=@n16.RGB
    Color_Night=@n16.RGB
   ) {materialFunction=&Color_Time_of_Day}.Result
   Alpha=Clamp(None=@n21.G)
  )
  MaterialFunctionCall(Emissive=MaterialFunctionCall(LDR Input=Clamp(None=@n22), HDR Tint=@n22, HDR Intensity Multiplier=ScalarParameter() {defaultValue=2, parameterName=Param_1}) {materialFunction=&DeriveHDRfromLDR}.RemappedRange) {materialFunction=&MF_FX_EmissiveAdjustments}
  n23 = ComponentMask(None=TextureCoordinate()) {r=true}
  n24 = ComponentMask(None=TextureCoordinate()) {g=true}
  n25 = Multiply(
   A=Power(Base=Sine(None=@n23) {period=2}, Exp=ScalarParameter() {defaultValue=2, parameterName=Side Edge Control})
   B=Power(Base=Sine(None=@n24) {period=2}, Exp=ScalarParameter() {defaultValue=5, parameterName=Top Edge Control})
  )
  Multiply(
   A=Multiply(A=ParticleColor().A, B=Multiply(A=@n21.RGB, B=@n25))
   B=MaterialFunctionCall(Distance=ScalarParameter() {defaultValue=150, parameterName=Near Fade Distance}) {materialFunction=&Near_Camera_Fade}.Result
  )
 ]
}
