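# Translucent, unlit, two-sided material for the Spectre skydiving ribbon effect, usable on Niagara and mesh particles.
# Panning caustic and smooth-tile textures, warped and blended by a pulsing normal-map Fresnel, mix time-of-day dark and bright emissive colors.
# Opacity combines inverse Fresnel, UV edge fade, depth fade and particle alpha, plus a deferred-only scene-normal intersection glow.
# Dynamic parameters drive sphere-masked, world-height noise scaled as an apparent vertex offset.
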Material /BRCosmetics/Athena/Effects/Materials/Skydiving/Spectre/M_Spectre_Ribbon
graph {
 properties {
  bUsedWithMeshParticles=true
  bUsedWithNiagaraMeshParticles=true
  bUsedWithNiagaraRibbons=true
  bUsedWithNiagaraSprites=true
  blendMode=BLEND_Translucent
  pixelDepthOffsetMode=PDOM_Legacy
  shadingModel=MSM_Unlit
  twoSided=true
  usedShadingModels=Unlit
 }
 nodes [
  n1 = VectorParameter() {parameterName=DarkColor, defaultValue=<rgba 0.125 0.3477 0.6 0>}
  n2 = MaterialFunctionCall(
   Color_Morning=@n1.RGB
   Color_Day=VectorParameter() {parameterName=DarkColor_Day, defaultValue=<rgba 0.025 0.0695 0.12 0>}.RGB
   Color_Evening=@n1.RGB
   Color_Night=@n1.RGB
  ) {materialFunction=&Color_Time_of_Day}
  n3 = VectorParameter() {parameterName=Color, defaultValue=<rgba 0.5 5 3.3215 1>}
  n4 = MaterialFunctionCall(
   Color_Morning=@n3.RGB
   Color_Day=VectorParameter() {parameterName=Color_Day, defaultValue=<rgba 0.7019 3 2.1428 1>}.RGB
   Color_Evening=@n3.RGB
   Color_Night=@n3.RGB
  ) {materialFunction=&Color_Time_of_Day}
  n5 = ScalarParameter() {defaultValue=0.333, parameterName=TextureTile}
  n6 = Multiply(A=Sine(None=Time()) {period=4}) {constB=0.12}
  n7 = Time()
  n8 = ScalarParameter() {defaultValue=10, parameterName=PanSpeed}
  n9 = TextureSampleParameter2D(UVs=Panner(Coordinate=TextureCoordinate() {vTiling=0.25}, Time=Multiply(A=@n7, B=@n8)) {speedX=0.01, speedY=0.06}) {parameterName=Fresnel_Normal, samplerType=SAMPLERTYPE_Normal, texture=/Game/Effects/Fort_Effects/Textures/Husk/Poison/T_PoisonTile_N}
  n10 = Transform(None=@n9.RGB)
  n11 = Fresnel(ExponentIn=Multiply(A=Add(B=@n6) {constA=0.7}, B=ScalarParameter() {defaultValue=0.4, parameterName=CausticsFresnelExp}), Normal=@n10) {baseReflectFraction=0, exponent=0.7}
  n12 = Multiply(A=@n11) {constB=0.1}
  n13 = Add(A=Multiply(A=TextureCoordinate() {uTiling=5.5, vTiling=3.6}, B=@n5), B=Multiply(A=@n12) {constB=-0.5})
  n14 = Multiply(A=Time(), B=ScalarParameter() {defaultValue=10, parameterName=PanSpeed})
  n15 = TextureCoordinate() {uTiling=2, vTiling=1.7}
  n16 = Panner(Coordinate=Add(A=@n12, B=Multiply(A=@n5, B=@n15)), Time=@n14) {speedX=0.01, speedY=0.06}
  n17 = Clamp(None=Power(Base=Multiply(A=@n11) {constB=1.5}) {constExponent=3})
  n18 = LinearInterpolate(
   A=TextureSample(UVs=Panner(Coordinate=@n13, Time=@n14) {speedX=-0.01, speedY=0.088}) {texture=&T_Smooth_tile}.G
   B=LinearInterpolate(Alpha=TextureSample(UVs=@n16) {texture=/Game/Global/Textures/Default/Caustics/T_Caustics_Softer}.G) {constA=0.4, constB=0.88}
   Alpha=@n17
  )
  n19 = DepthFade() {fadeDistanceDefault=80}
  n20 = ScalarParameter() {defaultValue=2, parameterName=WorldNormalFresnelExp}
  n21 = ComponentMask(None=SceneTexture() {sceneTextureId=PPI_WorldNormal}.Color) {b=true, g=true, r=true}
  n22 = Multiply(A=Multiply(A=OneMinus(None=@n19), B=Fresnel(ExponentIn=@n20, Normal=@n21) {baseReflectFraction=0, exponent=3})) {constB=0.2}
  n23 = Constant()
  n24 = ShadingPathSwitch(Default=@n22, Deferred=@n22, Forward=@n23, Mobile=@n23)
  n25 = Power(Base=Clamp(None=Add(A=@n18, B=@n24)) {maxDefault=2}, Exp=LinearInterpolate(Alpha=@n17) {constA=1.8, constB=0.8})
  n26 = Power(Base=@n25, Exp=ScalarParameter() {defaultValue=1, parameterName=EmissivePower})
  n27 = Clamp(None=@n26)
  MaterialFunctionCall(Emissive=Multiply(A=LinearInterpolate(A=@n2.Result, B=@n4.Result, Alpha=@n27), B=@n26)) {materialFunction=&MF_FX_EmissiveAdjustments}
  n28 = DynamicParameter() {defaultValue=<rgba 1 1 0 420>}
  n29 = Add(A=Multiply(A=Time(), B=@n28.Param3), B=WorldPosition().XYZ)
  n30 = Multiply(A=ObjectPositionWS(), B=Add(A=ScalarParameter() {parameterName=Speed}, B=@n28.Param1)) {constB=0.95}
  n31 = Divide(
   A=ComponentMask(None=Subtract(A=@n29, B=@n30)) {b=true}
   B=Add(A=ScalarParameter() {defaultValue=1100, parameterName=Tilling}, B=@n28.Param2)
  )
  n32 = WorldPosition()
  n33 = ObjectPositionWS()
  n34 = ScalarParameter() {defaultValue=2000}
  n35 = Constant()
  Multiply(
   A=Multiply(
    A=TextureSample(UVs=@n31) {texture=&T_NoisePack_01}.RGB
    B=DynamicParameter() {defaultValue=<rgba 1 1 0 420>}.Param4
   ) {constB=1500}
   B=OneMinus(None=SphereMask(A=@n32.XYZ, B=@n33, Radius=@n34, Hardness=@n35))
  )
  n36 = Add(A=@n25, B=ScalarParameter() {defaultValue=0.25, parameterName=DarknessAdd})
  n37 = ScalarParameter() {defaultValue=0.3, parameterName=OpacityFresnelExp}
  n38 = Multiply(
   A=OneMinus(None=Fresnel(ExponentIn=@n37, Normal=@n10) {baseReflectFraction=0})
   B=ScalarParameter() {defaultValue=8, parameterName=BodyOpacityMult}
  ) {constB=3}
  n39 = TextureCoordinate()
  n40 = Multiply(
   A=Add(A=Clamp(None=@n38), B=@n24)
   B=Power(Base=ComponentMask(None=@n39) {g=true}, Exp=ScalarParameter() {defaultValue=1, parameterName=Top Edge Fade})
  )
  Multiply(
   A=ParticleColor().A
   B=DepthFade(Opacity=Multiply(A=@n36, B=@n40), FadeDistance=ScalarParameter() {defaultValue=1, parameterName=FadeDistance})
  )
 ]
}
